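# Sound cue for the final demolition impact of a metal gym machine.
# Mixes three randomized layers: a pitched-down machine impact, a slightly delayed final-impact hit, and a default explosion impact layer.
# An outer modulator varies pitch on the whole mix for variation.
# Plays through medium-object destruction attenuation with concurrency limits on overlapping impacts.
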SoundCue /Game/Sounds/Fort_Impact_Sounds/Demolition/Gym/Demo_Metal_GymMachine_ImpactFinal_Cue
root=Modulator_9
nodes [
 {id=Delay_0, children [Random_17], properties {delayMax=0.3, delayMin=0.2}}
 {id=Mixer_4, children [Modulator_10,Delay_0,Random_20], properties {inputVolume [1,1,1]}}
 {id=Modulator_10, children [Random_16], properties {pitchMax=0.9, pitchMin=0.8}}
 {id=Modulator_9, children [Mixer_4], properties {pitchMax=1.07, pitchMin=0.92}}
 {id=Random_16, children [WavePlayer_51, WavePlayer_52, WavePlayer_53], properties {weights [1,1,1]}}
 {id=Random_17, children [WavePlayer_54, WavePlayer_55, WavePlayer_56], properties {weights [1,1,1]}}
 {id=Random_20, children [WavePlayer_72, WavePlayer_73], properties {weights [1,1]}}
 {id=WavePlayer_51, properties {soundWaveAssetPtr=&/Gym_Machine_Impact02}}
 {id=WavePlayer_52, properties {soundWaveAssetPtr=&/Gym_Machine_Impact03}}
 {id=WavePlayer_53, properties {soundWaveAssetPtr=&/Gym_Machine_Impact04}}
 {id=WavePlayer_54, properties {soundWaveAssetPtr=&/Gym_Machine_ImpactFinal03}}
 {id=WavePlayer_55, properties {soundWaveAssetPtr=&/Gym_Machine_ImpactFinal01}}
 {id=WavePlayer_56, properties {soundWaveAssetPtr=&/Gym_Machine_ImpactFinal02}}
 {id=WavePlayer_72, properties {soundWaveAssetPtr=/Game/Sounds/Fort_Impact_Sounds/Demolition/Defaults/Explo_Impact_Layer01}}
 {id=WavePlayer_73, properties {soundWaveAssetPtr=/Game/Sounds/Fort_Impact_Sounds/Demolition/Defaults/Explo_Impact_Layer02}}
]
properties {
 attenuationSettings=/Game/Sounds/Attenuation/Fort_Demolition_Destroy_Medium_Sized_Object_Att
 maxDistance=4000
 soundClassObject=&Fort_Impacts_PSM
 volumeMultiplier=0.408
 concurrencyOverrides {maxCount=3, resolutionRule=StopOldest}
 concurrencySet [&CONC_Object_Impacts_FourMax]
}
