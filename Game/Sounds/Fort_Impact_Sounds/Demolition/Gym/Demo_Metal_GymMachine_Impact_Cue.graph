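# Sound cue for demolition impacts on metal gym machines, playing in the impacts sound class.
# A random node picks one of three equally weighted gym machine impact waves, and a modulator above it slightly varies the pitch.
# It uses small melee weapon attenuation at reduced volume, with a shared object impact concurrency set alongside an override that stops the oldest instance.
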SoundCue /Game/Sounds/Fort_Impact_Sounds/Demolition/Gym/Demo_Metal_GymMachine_Impact_Cue
root=Modulator_20
nodes [
 {id=Modulator_20, children [Random_35], properties {pitchMax=1.07, pitchMin=0.92}}
 {id=Random_35, children [WavePlayer_124, WavePlayer_125, WavePlayer_126], properties {weights [1,1,1]}}
 {id=WavePlayer_124, properties {soundWaveAssetPtr=&/Gym_Machine_Impact02}}
 {id=WavePlayer_125, properties {soundWaveAssetPtr=&/Gym_Machine_Impact03}}
 {id=WavePlayer_126, properties {soundWaveAssetPtr=&/Gym_Machine_Impact04}}
]
properties {
 attenuationSettings=&Fort_Weapon_Melee_Sml_Att
 maxDistance=2800
 soundClassObject=&Fort_Impacts
 volumeMultiplier=0.34
 concurrencyOverrides {maxCount=3, resolutionRule=StopOldest}
 concurrencySet [&CONC_Object_Impacts_FourMax]
}
